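; Blueprint function library of UI conversion functions for the Season Pass screens, mapping reward view model and theme data to widget properties.
; Visibility helpers derive buy-pass, included-tag, quest-button and customize-icon states from ownership flags.
; Theme and reward helpers resolve background material, character color and first item details, with fallbacks for missing data.
; The quest-button check ANDs ownership with constant false, so it never evaluates true.
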(fn CF_BuyPassVisibility
  (params (ShowBuyPass :type Boolean) (HideYellowButtons :type Boolean) (CanUnlockSet :type Boolean))
  (return (select (select CanUnlockSet false (and ShowBuyPass (not HideYellowButtons))) "SelfHitTestInvisible" "Collapsed")))

(fn CF_GetPassBackgroundMaterial
  (params (ThemeData :type Season%20Pass%20Theme%20Data%20Structure))
  (bind (_titlecolor _titlematerial _titleoutline _seasoncolor _progressioncolor _overlaycolor _vaultbackgroundmaterial _giftboxvaultbackgroundmaterial _uibackgroundmaterial) (Utilities|Struct|BreakSeasonPassThemeData ThemeData))
  (if (Utilities|IsValidSoftObjectReference _uibackgroundmaterial)
    (return _uibackgroundmaterial)
    (else
      (return "/SeasonPass/Materials/RewardTrack/M_RewardTrack_TrayPlaceholder.M_RewardTrack_TrayPlaceholder"))))

(fn CF_GetQuestButtonVisibility
  (params (OwnsPass :type Boolean) (SelectedReward :type Season%20Pass%20Reward%20VM%20Object%20Reference))
  (Utilities|Casting|CastToAthenaSeasonItemEntryQuest (Class|SeasonPassRewardVM|GetReward SelectedReward :!class /Script/SeasonPass.SeasonPassRewardVM)
    (:then
      (return (Widget|ToVisibility(Boolean) (and OwnsPass false) "SelfHitTestInvisible")))
    (:CastFailed
      (return "Collapsed"))))

(fn CF_GetThemeCharacterColor
  (params (ThemeData :type Season%20Pass%20Category%20Theme%20Data%20Structure))
  (bind (_imagereferencetable _colorreferencetable) (Utilities|Struct|BreakSeasonPassElementThemeData ThemeData))
  (bind (_outcolor _returnvalue) (Class|FortColorReferenceFunctionLibrary|TryLookupColor _colorreferencetable "CharacterColor"))
  (return (select _returnvalue _outcolor "(R=0.092573,G=1.000000,B=0.000000,A=1.000000)")))

(fn GetFirstItemDetailsFromReward
  (params (SeasonPassReward :type Season%20Pass%20Reward%20VM%20Object%20Reference))
  (bind _itemvms (Variables|/Script/SeasonPass.SeasonPassRewardVM|GetItemVMs SeasonPassReward))
  (Utilities|IsValid SeasonPassReward
    (:Is%20Valid
      (if (> (Utilities|Array|Length _itemvms) 0)
        (return (Variables|/Script/FortniteUI.FortItemVM|GetItemDetails (Utilities|Array|Get(acopy) _itemvms)))
        (else
          (return 0))))
    (:Is%20Not%20Valid
      (return 0))))

(fn UpdateCustomizeIconVisibility
  (params (IsIncludedWithPass :type Boolean%20%28by%20ref%29) (DoesOwnPass :type Boolean%20%28by%20ref%29) (RequiresCustomization :type Boolean%20%28by%20ref%29))
  (return (or (and (not IsIncludedWithPass) RequiresCustomization) (Math|Boolean|ANDBoolean RequiresCustomization DoesOwnPass IsIncludedWithPass))))

(fn UpdateIsIncludedTagVisibility
  (params (IsPassOwned :type Boolean%20%28by%20ref%29) (isIncludedWithPass :type Boolean%20%28by%20ref%29) (isRewardOwned :type Boolean%20%28by%20ref%29))
  (return (select IsPassOwned "Collapsed" (select isIncludedWithPass (select isRewardOwned "Collapsed" "SelfHitTestInvisible") "Collapsed"))))
