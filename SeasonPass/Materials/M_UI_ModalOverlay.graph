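# M_UI_ModalOverlay is a translucent UI-domain material for Season Pass, named as a modal overlay.
# A Corner Radius vector parameter, its RGB appended to itself, feeds the MF_UI_CoreUI_RoundedMask function, with a constant driving its Radius in Pixels input.
# A separate constant color node sits in the graph, and no links to the material's output pins appear.
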Material /SeasonPass/Materials/M_UI_ModalOverlay
graph {
 properties {
  bAutomaticallySetUsageInEditor=false
  bUsedWithStaticMesh=false
  blendMode=BLEND_Translucent
  materialDomain=MD_UI
  usedShadingModels=Default Lit
 }
 nodes [
  Constant3Vector()
  n1 = VectorParameter() {parameterName=Corner Radius, defaultValue=<rgba 20 20 20 20>}
  MaterialFunctionCall(Corner Radius=AppendVector(A=@n1.RGB, B=@n1.RGB), Radius in Pixels=Constant() {r=1}) {materialFunction=&MF_UI_CoreUI_RoundedMask}
 ]
}
