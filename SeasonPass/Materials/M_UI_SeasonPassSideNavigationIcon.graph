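# UI-domain translucent material named for the side navigation icon in the Season Pass screen.
# Its graph holds a single scalar parameter, Transparent, with no connection shown to any material output.
# Automatic usage setting in the editor and static mesh usage are both turned off.
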Material /SeasonPass/Materials/M_UI_SeasonPassSideNavigationIcon
graph {
 properties {
  bAutomaticallySetUsageInEditor=false
  bUsedWithStaticMesh=false
  blendMode=BLEND_Translucent
  materialDomain=MD_UI
  usedShadingModels=Default Lit
 }
 nodes [ScalarParameter() {parameterName=Transparent}]
}
